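; Cosmetic trail actor that draws spline meshes behind a skeletal mesh socket, with optional particles following it.
; Each tick moves the last spline point to the socket, adds points by distance and expires old ones by timestamp into a mesh pool.
; Cheap mode passes positions and timestamps to materials, while full mode samples size and color curves.
; Timers drive camera-distance tick LODs, render checks and timed shutdown; a manual stop also exists.
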(event EventBeginPlay
  (bind _splinesystemduration (Variables|Default|GetSplineSystemDuration))
  (bind _skeletal_mesh (Variables|Components|GetSkeletal%20Mesh))
  (bind _particles (Variables|Default|GetParticles))
  (bind _tick_interval_event_handle (Variables|LODs|GetTick%20Interval%20Event%20Handle))
  ;[1 1 1] Begin Play Stuff
  ;;[1 1 1] Duration of the spline system if X > 0
  (if (> _splinesystemduration 0.0)
    (bind _returnvalue (Utilities|Time|SetTimerbyEvent (delegate Custom|SplineSystemDurationReached) _splinesystemduration))
    (Variables|LODs|SetSpline%20System%20Duration%20Event%20Handle _returnvalue)
    (else
      _))
  (Networking|CanExecuteCosmeticEvents
    (:True
      (Utilities|IsValid _skeletal_mesh
        (:Is%20Valid
          (if (Transformation|DoesSocketExist _skeletal_mesh (Variables|Default|GetSocket%20Attach%20Point))
            (Variables|Business|SetSplineSystemActive true)
            (if (Variables|Debug|GetDebugMode)
              (CallFunction|DebugModeRotate))
            (Utilities|IsValid _particles
              (:Is%20Valid
                (bind _returnvalue_2 (Effects|SpawnEmitteratLocation _particles (CallFunction|GetSocketLocationWS (Variables|Default|GetSocket%20Attach%20Point))))
                (Variables|Business|SetParticlesComponent _returnvalue_2))
              (:Is%20Not%20Valid
                _))
            ;;[1 1 1] Tick Interval LODs
            (if (Variables|LODs|GetUse%20Tick%20Interval%20LODs)
              (bind _returnvalue_3 (Utilities|Time|SetTimerbyEvent (delegate Custom|Tick%20Interval%20LOD%20Check) (Variables|LODs|GetTick%20Interval%20LOD%20Check%20Rate) true))
              (Variables|LODs|SetTick%20Interval%20Event%20Handle _returnvalue_3)
              (else
                _))
            ;;[1 1 1] Recently Rendered LOD
            (if (Variables|LODs|GetUse%20Recently%20Rendered%20LOD)
              (bind _returnvalue_4 (Utilities|Time|SetTimerbyEvent (delegate Custom|Recently%20Rendered%20Check) (Variables|LODs|GetRecently%20Rendered%20Check%20Rate) true))
              (Variables|LODs|SetRecently%20Rendered%20Event%20Handle _returnvalue_4))
            (else
              (Development|PrintString "Could not find socket to attach to, aborting" true true "(R=1.000000,G=0.000000,B=0.168010,A=1.000000)" 10))))
        (:Is%20Not%20Valid
          (Development|PrintString "Could not find valid skeletal mesh, aborting")))
      ;[1 1 1] Timer / Destroy
      (if (Utilities|Time|IsValidTimerHandle _tick_interval_event_handle)
        (Utilities|Time|ClearandInvalidateTimerbyHandle _tick_interval_event_handle)
        (else
          _))
      (Actor|Tick|SetActorTickEnabled false)
      (CallFunction|DestroyAllSplineMeshes)
      (Actor|DestroyActor))))

(event Custom|Recently%20Rendered%20Check
  ;[1 1 1] Begin Play Stuff
  ;;[1 1 1] Recently Rendered LOD
  (CallFunction|SetWasRecentlyRendered))

(event Custom|Tick%20Interval%20LOD%20Check
  ;[1 1 1] Begin Play Stuff
  ;;[1 1 1] Tick Interval LODs
  (CallFunction|TickIntervalLODCheck))

(event Custom|SplineSystemDurationReached
  (bind _tick_interval_event_handle (Variables|LODs|GetTick%20Interval%20Event%20Handle))
  ;[1 1 1] Timer / Destroy
  (Variables|Business|SetSplineSystemActive false)
  (Variables|Business|SetCompletionTimestamp (Utilities|Time|GetGameTimeinSeconds))
  (CallFunction|SetNewSplinePointsBool)
  (Actor|Tick|SetActorTickEnabled (Variables|Default|GetKeepTickEnabledWhenCompleting))
  (Utilities|FlowControl|Delay (Variables|Default|GetLifetimeAfterCompletion))
  (if (Utilities|Time|IsValidTimerHandle _tick_interval_event_handle)
    (Utilities|Time|ClearandInvalidateTimerbyHandle _tick_interval_event_handle)
    (else
      _))
  (Actor|Tick|SetActorTickEnabled false)
  (CallFunction|DestroyAllSplineMeshes)
  (Actor|DestroyActor))

(event Custom|ManuallyStopSplineSystem
  (bind _tick_interval_event_handle (Variables|LODs|GetTick%20Interval%20Event%20Handle))
  ;[1 1 1] Timer / Destroy
  (if (Variables|Business|GetSplineSystemActive)
    (Utilities|Time|ClearandInvalidateTimerbyHandle (Variables|LODs|GetSpline%20System%20Duration%20Event%20Handle))
    (Variables|Business|SetSplineSystemActive false)
    (Variables|Business|SetCompletionTimestamp (Utilities|Time|GetGameTimeinSeconds))
    (CallFunction|SetNewSplinePointsBool)
    (Actor|Tick|SetActorTickEnabled (Variables|Default|GetKeepTickEnabledWhenCompleting))
    (Utilities|FlowControl|Delay (Variables|Default|GetLifetimeAfterCompletion))
    (if (Utilities|Time|IsValidTimerHandle _tick_interval_event_handle)
      (Utilities|Time|ClearandInvalidateTimerbyHandle _tick_interval_event_handle)
      (else
        _))
    (Actor|Tick|SetActorTickEnabled false)
    (CallFunction|DestroyAllSplineMeshes)
    (Actor|DestroyActor)))

(event EventTick
  (params (DeltaSeconds :type Float%20%28single-precision%29))
  ;[0.995 0.394 0.0596 0.5] Tick
  (CallFunction|SplineUpdateonTick))

(event Custom|DebugMode_Rotate
  (bind _returnvalue (Utilities|Time|GetGameTimeinSeconds))
  (bind _splinetimestamps (Variables|Business|GetSplineTimestamps))
  ;[1 1 1] Debug Mode
  ;;[1 1 1] Timestamps from construction script are bad . . .possible bug? We need to set these so don't get 2 splines just hanging around
  (Utilities|Array|SetArrayElem _splinetimestamps 0 (- _returnvalue 0.5))
  (Utilities|Array|SetArrayElem _splinetimestamps 1 (- _returnvalue 0.25))
  (Variables|Business|SetDebugPreviousTickTimestamp (Utilities|Time|GetGameTimeinSeconds))
  (Timeline|DebugRotateTimeline PlayFromStart
    (:Update
      (Variables|Business|SetDebugTickDelta (- (Utilities|Time|GetGameTimeinSeconds) (Variables|Business|GetDebugPreviousTickTimestamp)))
      (Transformation|AddLocalRotation (Variables|Components|GetDebugArrow) (Math|Rotator|ScaleRotator(Float) (Variables|Debug|GetDebugRotationRate) (Variables|Business|GetDebugTickDelta)) false false)
      (Variables|Business|SetDebugPreviousTickTimestamp (Utilities|Time|GetGameTimeinSeconds)))))

(event Collision|EventActorBeginOverlap
  (params (OtherActor :type Actor%20Object%20Reference)))

(fn AddSplineDistanceCheck
  (return (and (>= (Math|Vector|VectorLength (- (Spline|GetLocationatSplinePoint (Variables|Components|GetSpline) (- (Spline|GetNumberOfSplinePoints (Variables|Components|GetSpline)) 1) "World") (Spline|GetLocationatSplinePoint (Variables|Components|GetSpline) (- (Spline|GetNumberOfSplinePoints (Variables|Components|GetSpline)) 2) "World"))) (select (== (Utilities|Array|Length (Variables|Business|GetSplineMeshes)) 0) (Variables|Default|GetSplineDistanceThresholdEmpty) (Variables|Default|GetSplineDistanceThreshold))) (Variables|Business|GetCanNewSplinePointsBeCreated))))

(fn AddSplineMesh
  (bind _splinetimestamps (Variables|Business|GetSplineTimestamps))
  (bind _returnvalue (Utilities|Array|LastIndex _splinetimestamps))
  (bind _mids_0 (Variables|Business|GetMIDs_0))
  (bind _returnvalue_2 (Utilities|Array|LastIndex _mids_0))
  (bind _spline (Variables|Components|GetSpline))
  (bind _returnvalue_3 (Spline|GetNumberOfSplinePoints _spline))
  (bind _returnvalue_4 (- _returnvalue_3 2))
  (bind _spline_2 (Variables|Components|GetSpline))
  (bind _returnvalue_5 (Spline|GetLocationatSplinePoint _spline_2 _returnvalue_4 "World"))
  (bind _returnvalue_6 (- _returnvalue_2 1))
  (bind _splinemesheshidden (Variables|Business|GetSplineMeshesHidden))
  (bind _output (Utilities|Array|Get(acopy) _splinemesheshidden))
  (bind _output_2 (Utilities|ClampMinInteger _returnvalue_6))
  (bind _output_3 (Utilities|Array|Get(acopy) _mids_0 _output_2))
  (bind _current_spline_mesh (Variables|Business|GetCurrent%20Spline%20Mesh))
  (bind _splinemeshtemplate (Variables|Default|GetSplineMeshTemplate))
  (Math|Integer|IncrementInt (Variables|Business|GetSplineID)
    (:%20%20
      ;[1 1 1] Check if we can get a spline mesh from the pool
      (if (< (Utilities|Array|Length _splinemesheshidden) 1)
        ;[1 1 1] No, create new spline mesh
        (bind _returnvalue_7 (AddComponent|NODE_AddSplineMeshComponent-0 :bManualAttachment true))
        (bind _output_get (Variables|Business|SetCurrent%20Spline%20Mesh _returnvalue_7))
        (Utilities|Array|Add (Variables|Business|GetSplineMeshes) _output_get)
        (CallFunction|SetForwardAxisonSplineMesh (Variables|Business|GetCurrent%20Spline%20Mesh))
        ;[1 1 1] Set static mesh template, create MID, set splineID Param
        (Utilities|IsValid _splinemeshtemplate
          (:Is%20Valid
            (Components|StaticMesh|SetStaticMesh _current_spline_mesh _splinemeshtemplate)
            (bind _returnvalue_8 (Rendering|Material|CreateDynamicMaterialInstance _current_spline_mesh 0 (select (Utilities|IsValid (Variables|Default|GetSplineMeshMaterialOverride) :!class /Script/Engine.KismetSystemLibrary) (Variables|Default|GetSplineMeshMaterialOverride) (Rendering|Material|GetMaterial _current_spline_mesh 0)) "None" :!class /Script/Engine.PrimitiveComponent))
            (Utilities|Array|Add (Variables|Business|GetMIDs_0) _returnvalue_8)
            (Rendering|Material|SetScalarParameterValue _returnvalue_8 "SplineID" (Math|Conversions|ToFloat(Integer) (Variables|Business|GetSplineID)) :!class /Script/Engine.MaterialInstanceDynamic)
            (if (Variables|Default|GetUseCheapSplines)
              ;[1 1 1] Cheap spline material stuff
              (Rendering|Material|SetScalarParameterValue _returnvalue_8 "SplineTimestampA" (Utilities|Array|Get(acopy) _splinetimestamps _returnvalue) :!class /Script/Engine.MaterialInstanceDynamic)
              (Rendering|Material|SetScalarParameterValue _returnvalue_8 "SplineTimestampB" (Utilities|Array|Get(acopy) _splinetimestamps (- _returnvalue 1)) :!class /Script/Engine.MaterialInstanceDynamic)
              (Rendering|Material|SetScalarParameterValue _returnvalue_8 "SplineMeshLifetime" (Variables|Default|GetLifetime) :!class /Script/Engine.MaterialInstanceDynamic)
              (Rendering|Material|SetVectorParameterValue _returnvalue_8 "WP_B" (Math|Conversions|ToLinearColor(Vector) _returnvalue_5) :!class /Script/Engine.MaterialInstanceDynamic)
              (Utilities|IsValid _output_3
                (:Is%20Valid
                  (Rendering|Material|SetVectorParameterValue _output_3 "WP_A" (Math|Conversions|ToLinearColor(Vector) (Spline|GetLocationatSplinePoint (Variables|Components|GetSpline) (- _returnvalue_3 1) "World")) :!class /Script/Engine.MaterialInstanceDynamic)
                  (Utilities|IsValid (Variables|Default|GetLinearColorTexture)
                    (:Is%20Valid
                      (Rendering|Material|SetTextureParameterValue _returnvalue_8 "LinearColorTexture" (Variables|Default|GetLinearColorTexture))
                      ;[1 1 1] Need to make sure we have enough spline meshes
                      (if (< (Utilities|Array|Length (Variables|Business|GetSplineMeshes)) (- (Spline|GetNumberOfSplinePoints (Variables|Components|GetSpline)) 1))
                        _)
                      (return))
                    (:Is%20Not%20Valid
                      _))
                  (Development|PrintString "Linear color Texture variable not set, please fix :)" true true "(R=0.000000,G=0.660000,B=1.000000,A=1.000000)" 10))
                (:Is%20Not%20Valid
                  _))
              (if (== _output_2 0)
                _)
              (Rendering|Material|SetVectorParameterValue _returnvalue_8 "WP_A" (Math|Conversions|ToLinearColor(Vector) _returnvalue_5) :!class /Script/Engine.MaterialInstanceDynamic))
            (Rendering|Material|SetScalarParameterValue _returnvalue_8 "CheapSplines" (Math|Conversions|ToFloat(Boolean) (Variables|Default|GetUseCheapSplines)) :!class /Script/Engine.MaterialInstanceDynamic))))
      (Utilities|IsValid _output
        (:Is%20Valid
          ;[1 1 1] Yes, get spline mesh from the pool
          (bind _output_get_2 (Variables|Business|SetCurrent%20Spline%20Mesh _output))
          (Rendering|SetVisibility _output_get_2 true false)
          (Utilities|Array|Add (Variables|Business|GetSplineMeshes) (Variables|Business|GetCurrent%20Spline%20Mesh))
          (Utilities|Array|RemoveIndex (Variables|Business|GetSplineMeshesHidden)))
        (:Is%20Not%20Valid
          _))
      (Utilities|Array|RemoveIndex (Variables|Business|GetSplineMeshesHidden)))))

(fn AddSplinePoint
  (Spline|AddSplinePoint (Variables|Components|GetSpline) (CallFunction|GetSocketLocationWS (Variables|Default|GetSocket%20Attach%20Point)) "World" true)
  (Utilities|Array|Add (Variables|Business|GetSplineTimestamps) (Utilities|Time|GetGameTimeinSeconds))
  (if (and (> (Spline|GetNumberOfSplinePoints (Variables|Components|GetSpline)) 2) (Variables|Default|GetRedistributePointsAlongSplineWhenCreated))
    (CallFunction|RedistributePointsAlongSpline)))

(fn BakeLinearColor
  (params (Curve%20Linear%20Color :type Curve%20Linear%20Color%20Object%20Reference) (Pixel%20%28X%29 :type Integer))
  (bind (_r _g _b _a) (Utilities|Struct|BreakLinearColor (Math|Curves|GetLinearColorValue Curve%20Linear%20Color (+ (/ (Math|Conversions|ToFloat(Integer) Pixel%20%28X%29) (Math|Vector2D|BreakVector2D (Variables|CurveBaking|GetRenderTextureResolution))) (/ 1 (Math|Vector2D|BreakVector2D (Variables|CurveBaking|GetRenderTextureResolution)))))))
  (return (Utilities|Struct|MakeLinearColor _r _g _b (- 1 _a))))

(fn DestroyAllSplineMeshes
  (for _array_element (Variables|Business|GetSplineMeshes)
    (Utilities|IsValid _array_element
      (:Is%20Valid
        (Components|DestroyComponent _array_element))))
  (for _array_element_2 (Variables|Business|GetSplineMeshesHidden)
    (Utilities|IsValid _array_element_2
      (:Is%20Valid
        (Components|DestroyComponent _array_element_2))))
  (return))

(fn GetLinearColorFromCurveAtTime
  (params (Curve%20Linear%20Color :type Curve%20Linear%20Color%20Object%20Reference) (Index :type Integer))
  (return (Math|Curves|GetLinearColorValue Curve%20Linear%20Color (select (Variables|Default|GetLinearColorOverLengthOfSpline) (/ (Math|Conversions|ToFloat(Integer) Index) (Math|Conversions|ToFloat(Integer) (- (Spline|GetNumberOfSplinePoints (Variables|Components|GetSpline)) 1))) (/ (- (Utilities|Time|GetGameTimeinSeconds) (Utilities|Array|Get(acopy) (Variables|Business|GetSplineTimestamps) Index)) (Variables|Business|GetLifetimeAdjusted))))))

(fn GetSocketLocationWS
  (params (Socket%20Name :type Name))
  (bind _skeletal_mesh (Variables|Components|GetSkeletal%20Mesh))
  (Utilities|IsValid _skeletal_mesh
    (:Is%20Valid
      (return (select (Transformation|DoesSocketExist _skeletal_mesh Socket%20Name) (Transformation|GetSocketLocation _skeletal_mesh Socket%20Name) (Transformation|GetWorldLocation _skeletal_mesh))))))

(fn RedistributePointsAlongSpline
  (bind _spline (Variables|Components|GetSpline))
  (bind _returnvalue (Spline|GetNumberOfSplinePoints _spline))
  (bind _returnvalue_2 (- _returnvalue 1))
  (for _index (range _returnvalue)
    (Spline|SetLocationatSplinePoint (Variables|Components|GetSpline) _index (Math|Vector|Lerp(Vector) (Spline|GetLocationatSplinePoint (Variables|Components|GetSpline) _index "World") (Spline|GetLocationatTime (Variables|Components|GetSpline) (/ (Math|Conversions|ToFloat(Integer) _index) (Math|Conversions|ToFloat(Integer) _returnvalue_2)) "World" true) (Math|Float|Power (/ (Math|Conversions|ToFloat(Integer) _index) (Math|Conversions|ToFloat(Integer) _returnvalue_2)) 5)) "World" false))
  (Spline|UpdateSpline (Variables|Components|GetSpline)))

(fn ReturnSplineMeshToPool
  (params (SplineMesh :type Spline%20Mesh%20Component%20Object%20Reference) (IndexToRemove :type Integer))
  (Rendering|SetVisibility SplineMesh false false)
  (Utilities|Array|RemoveIndex (Variables|Business|GetSplineMeshes) IndexToRemove)
  (Utilities|Array|Add (Variables|Business|GetSplineMeshesHidden) SplineMesh)
  (Utilities|Array|RemoveIndex (Variables|Business|GetMIDs_0) IndexToRemove))

(fn RollSplineMesh
  (Utilities|IsValid (Variables|Curves|GetRotationOverLife)
    (:Is%20Valid
      (SplineMesh|SetStartRollRadians (Variables|Business|GetCurrent%20Spline%20Mesh) (* (Variables|Default|GetInitial%20Rotation%20%28Roll%29) (Utilities|GetCurveFloatValueatTime (Variables|Curves|GetRotationOverLife) (Variables|Business|GetArrayIndex))) false)
      (SplineMesh|SetEndRollRadians (Variables|Business|GetCurrent%20Spline%20Mesh) (* (Variables|Default|GetInitial%20Rotation%20%28Roll%29) (Utilities|GetCurveFloatValueatTime (Variables|Curves|GetRotationOverLife) (+ (Variables|Business|GetArrayIndex) 1))) false))))

(fn ScaleSplineMesh
  (bind (_r _g) (Math|Color|BreakColor (CallFunction|GetLinearColorfromCurveatTime (Variables|Curves|GetSizeOverLife) (Variables|Business|GetArrayIndex))))
  (bind (_r_2 _g_2) (Math|Color|BreakColor (CallFunction|GetLinearColorfromCurveatTime (Variables|Curves|GetSizeOverLife) (+ (Variables|Business|GetArrayIndex) 1))))
  (Utilities|IsValid (Variables|Curves|GetSizeOverLife)
    (:Is%20Valid
      (SplineMesh|SetStartScale (Variables|Business|GetCurrent%20Spline%20Mesh) (* (Math|Vector2D|MakeVector2D _r _g) (Variables|Default|GetInitial%20Size)) false :!class /Script/Engine.SplineMeshComponent)
      (SplineMesh|SetEndScale (Variables|Business|GetCurrent%20Spline%20Mesh) (* (Math|Vector2D|MakeVector2D _r_2 _g_2) (Variables|Default|GetInitial%20Size)) false :!class /Script/Engine.SplineMeshComponent)
      (return))))

(fn SetForwardAxisOnSplineMesh
  (params (SplineMesh :type Spline%20Mesh%20Component%20Object%20Reference))
  (switch Utilities|FlowControl|Switch|SwitchonEn_SplineForwardAxes_01 (Variables|Default|GetSplineMeshForwardAxis)
    (:NewEnumerator0
      (SplineMesh|SetForwardAxis SplineMesh "X" false))
    (:NewEnumerator1
      (SplineMesh|SetForwardAxis SplineMesh "Y" false))
    (:NewEnumerator2
      (SplineMesh|SetForwardAxis SplineMesh "Z" false))))

(fn SetNewSplinePointsBool
  (Variables|Business|SetCanNewSplinePointsBeCreated (not (Variables|Default|GetCeaseAndDetachSplineUponCompletion))))

(fn SetStartAndEndAllSplineMeshes
  (params (Update%20All%20Spline%20Meshes :type Boolean))
  (for (_array_element _array_index) (Variables|Business|GetSplineMeshes)
    (Variables|Business|SetArrayIndex _array_index)
    (Variables|Business|SetCurrent%20Spline%20Mesh _array_element)
    (CallFunction|SetStartandEndSplineMesh Update%20All%20Spline%20Meshes)))

(fn SetStartAndEndSplineMesh
  (params (Update%20Spline%20Mesh :type Boolean))
  (bind (_location _tangent) (Spline|GetLocationandTangentatSplinePoint (Variables|Components|GetSpline) (+ (Variables|Business|GetArrayIndex) 1) "World"))
  (bind (_location_2 _tangent_2) (Spline|GetLocationandTangentatSplinePoint (Variables|Components|GetSpline) (Variables|Business|GetArrayIndex) "World"))
  (SplineMesh|SetStartandEnd (Variables|Business|GetCurrent%20Spline%20Mesh) _location_2 _tangent_2 _location _tangent Update%20Spline%20Mesh :!class /Script/Engine.SplineMeshComponent)
  (return))

(fn SetWasRecentlyRendered
  (bind _recently_rendered_check_rate (Variables|LODs|GetRecently%20Rendered%20Check%20Rate))
  (bind _returnvalue (Rendering|WasActorRecentlyRendered _recently_rendered_check_rate))
  (Utilities|FlowControl|Sequence
    (:then_0
      (if (and (Math|Boolean|NotEqual(Boolean) _returnvalue (Variables|LODs|GetRecentlyRendered)) (not (Variables|LODs|GetRecentlyRendered)))
        (CallFunction|SplineTimestampManagement)
        (CallFunction|SetStartandEndAllSplineMeshes true)))
    (:then_1
      (Variables|LODs|SetRecentlyRendered _returnvalue))))

(fn SplineTimestampManagement
  (bind _splinemeshes (Variables|Business|GetSplineMeshes))
  (if (not (Variables|Business|GetSplineSystemActive))
    ;[1 1 1] Lifetime after we deactivate, but still creating new splines (lifetime after completion)
    (Variables|Business|SetLifetimeAdjusted (Math|Float|Lerp (Variables|Default|GetLifetime) 0.01 (Math|Float|Clamp(Float) (/ (- (Utilities|Time|GetGameTimeinSeconds) (Variables|Business|GetCompletionTimestamp)) (Variables|Default|GetLifetimeAfterCompletion))))))
  (Utilities|Array|Clear (Variables|Business|GetIndiciesToDelete))
  (for (_array_element _array_index) (Variables|Business|GetSplineTimestamps)
    (Variables|Business|SetArrayIndex _array_index)
    (if (>= (- (- (Utilities|Time|GetGameTimeinSeconds) _array_element) (Variables|Default|GetLifetime%20Buffer)) (Variables|Business|GetLifetimeAdjusted))
      (Utilities|Array|Add (Variables|Business|GetIndiciesToDelete) (Variables|Business|GetArrayIndex))
      (if (> (Spline|GetNumberOfSplinePoints (Variables|Components|GetSpline)) 2)
        (Spline|RemoveSplinePoint (Variables|Components|GetSpline) (Variables|Business|GetArrayIndex) false))))
  (for _array_element_2 (Variables|Business|GetIndiciesToDelete)
    (if (> (Utilities|Array|Length (Variables|Business|GetSplineTimestamps)) 2)
      (Utilities|Array|RemoveIndex (Variables|Business|GetSplineTimestamps) _array_element_2)
      (else
        _))
    (if (Utilities|Array|IsValidIndex (Variables|Business|GetSplineMeshes) _array_element_2)
      (Utilities|IsValid (Utilities|Array|Get(acopy) _splinemeshes _array_element_2)
        (:Is%20Valid
          (CallFunction|ReturnSplineMeshtoPool (Utilities|Array|Get(acopy) _splinemeshes _array_element_2) _array_element_2)))))
  (Spline|UpdateSpline (Variables|Components|GetSpline))
  (return))

(fn SplineUpdateOnTick
  (bind _mids_0 (Variables|Business|GetMIDs_0))
  (bind _returnvalue (Utilities|Array|LastIndex _mids_0))
  (Networking|CanExecuteCosmeticEvents
    (:True
      ;[1 1 1] If a particle system is used, update its location
      (CallFunction|UpdateParticleSystemPosition)
      ;[1 1 1] Sets the position of the last spline point to the socket's position in WS
      (CallFunction|UpdateSplinePointPosition)
      ;[1 1 1] Lifetime management - if the spline's timestamp exceeds the lifetime variable, destroy the spline point and mesh
      (CallFunction|SplineTimestampManagement)
      (if (Variables|LODs|GetRecentlyRendered)
        ;[0.5681 1 0.6964 0.1] Distance check - if threshold is exceeded, add new spline point and spline mesh
        (if (CallFunction|AddSplineDistanceCheck)
          ;[1 1 1] Add spline point and new spline mesh
          (CallFunction|AddSplinePoint)
          (CallFunction|AddSplineMesh)
          (CallFunction|UpdateCheapSplinesWPOVectors)
          (else
            _))
        (if (Variables|Default|GetUseCheapSplines)
          ;[1 1 1] "Cheap" Splines
          (for _index (range (Utilities|ClampMinInteger (- (Utilities|Array|LastIndex (Variables|Business|GetSplineMeshes)) (Variables|Default|GetNumberOfSplineMeshesToUpdate))) (+ (Utilities|Array|LastIndex (Variables|Business|GetSplineMeshes)) 1))
            (bind _output_get (Variables|Business|SetArrayIndex _index))
            (Variables|Business|SetCurrent%20Spline%20Mesh (Utilities|Array|Get(acopy) (Variables|Business|GetSplineMeshes) _output_get))
            (if (== _returnvalue (Variables|Business|GetArrayIndex))
              (Rendering|Material|SetVectorParameterValue (Utilities|Array|Get(acopy) _mids_0 _returnvalue) "WP_A" (Math|Conversions|ToLinearColor(Vector) (Spline|GetLocationatSplinePoint (Variables|Components|GetSpline) (- (Spline|GetNumberOfSplinePoints (Variables|Components|GetSpline)) 1) "World")) :!class /Script/Engine.MaterialInstanceDynamic)
              (else
                _))
            (CallFunction|SetStartandEndSplineMesh true))
          (else
            ;[1 1 1] This updates all spline meshes start/end positions, samples curve data on tick and sets on MIDs on tick. Pretty expensive.
            (for (_array_element _array_index) (Variables|Business|GetSplineMeshes)
              (Variables|Business|SetArrayIndex _array_index)
              (Variables|Business|SetCurrent%20Spline%20Mesh _array_element)
              (CallFunction|ScaleSplineMesh)
              (CallFunction|UpdateMID)
              (CallFunction|SetStartandEndSplineMesh true))))))))

(fn TickIntervalLODCheck
  (bind _tick_interval_lod_distances (Variables|LODs|GetTick%20Interval%20LOD%20Distances))
  (bind _returnvalue (Utilities|Array|LastIndex _tick_interval_lod_distances))
  (bind _index (Utilities|FlowControl|ForLoopwithBreak 0 _returnvalue)
    (:LoopBody
      (Variables|Business|SetLOD%20Array%20Index _index)
      (if (or (and (>= (Math|Vector|VectorLength (- (CallFunction|GetSocketLocationWS (Variables|Default|GetSocket%20Attach%20Point)) (Transformation|GetActorLocation (Game|GetPlayerCameraManager 0)))) (Math|Conversions|ToFloat(Integer) (Utilities|Array|Get(acopy) (Variables|LODs|GetTick%20Interval%20LOD%20Distances) (Variables|Business|GetLOD%20Array%20Index)))) (<= (Math|Vector|VectorLength (- (CallFunction|GetSocketLocationWS (Variables|Default|GetSocket%20Attach%20Point)) (Transformation|GetActorLocation (Game|GetPlayerCameraManager 0)))) (Math|Conversions|ToFloat(Integer) (Utilities|Array|Get(acopy) (Variables|LODs|GetTick%20Interval%20LOD%20Distances) (Utilities|ClampMaxInteger (+ (Variables|Business|GetLOD%20Array%20Index) 1) _returnvalue))))) (and (> (Math|Vector|VectorLength (- (CallFunction|GetSocketLocationWS (Variables|Default|GetSocket%20Attach%20Point)) (Transformation|GetActorLocation (Game|GetPlayerCameraManager 0)))) (Math|Conversions|ToFloat(Integer) (Utilities|Array|Get(acopy) (Variables|LODs|GetTick%20Interval%20LOD%20Distances) (Utilities|Array|LastIndex (Variables|LODs|GetTick%20Interval%20LOD%20Distances))))) (== (Utilities|Array|LastIndex (Variables|LODs|GetTick%20Interval%20LOD%20Distances)) (Variables|Business|GetLOD%20Array%20Index))))
        (Utilities|FlowControl|Sequence
          (:then_0
            _)
          (:then_1
            _))
        (Actor|Tick|SetActorTickInterval (Utilities|Array|Get(acopy) (Variables|LODs|GetTick%20Actor%20Intervals) (Variables|Business|GetLOD%20Array%20Index)))))))

(fn UpdateCheapSplinesWPOVectors
  (bind _spline (Variables|Components|GetSpline))
  (bind _returnvalue (Spline|GetNumberOfSplinePoints _spline))
  (bind _returnvalue_2 (- _returnvalue 1))
  (bind _returnvalue_3 (Math|Conversions|ToFloat(Integer) _returnvalue_2))
  (if (Variables|Default|GetUseCheapSplines)
    (for (_array_element _array_index) (Variables|Business|GetMIDs_0)
      (Rendering|Material|SetVectorParameterValue _array_element "WP_A" (Math|Conversions|ToLinearColor(Vector) (Spline|GetLocationatSplinePoint (Variables|Components|GetSpline) (+ _array_index 1) "World")) :!class /Script/Engine.MaterialInstanceDynamic)
      (Rendering|Material|SetVectorParameterValue _array_element "WP_B" (Math|Conversions|ToLinearColor(Vector) (Spline|GetLocationatSplinePoint (Variables|Components|GetSpline) _array_index "World")) :!class /Script/Engine.MaterialInstanceDynamic)
      (if (Variables|Default|GetLinearColorOverLengthOfSpline)
        (Rendering|Material|SetScalarParameterValue _array_element "ManualB" (/ (Math|Conversions|ToFloat(Integer) _array_index) _returnvalue_3) :!class /Script/Engine.MaterialInstanceDynamic)
        (Rendering|Material|SetScalarParameterValue _array_element "ManualA" (/ (Math|Conversions|ToFloat(Integer) (+ _array_index 1)) _returnvalue_3) :!class /Script/Engine.MaterialInstanceDynamic)))))

(fn UpdateMID
  (bind _arrayindex (Variables|Business|GetArrayIndex))
  (bind _mids_0 (Variables|Business|GetMIDs_0))
  (bind _output (Utilities|Array|Get(acopy) _mids_0 _arrayindex))
  (Utilities|IsValid (Variables|Curves|GetColorAndAlphaOverLife)
    (:Is%20Valid
      (Rendering|Material|SetVectorParameterValue _output "LinearColorA" (CallFunction|GetLinearColorfromCurveatTime (Variables|Curves|GetColorAndAlphaOverLife) (Variables|Business|GetArrayIndex)) :!class /Script/Engine.MaterialInstanceDynamic)
      (Rendering|Material|SetVectorParameterValue _output "LinearColorB" (CallFunction|GetLinearColorfromCurveatTime (Variables|Curves|GetColorAndAlphaOverLife) (+ (Variables|Business|GetArrayIndex) 1)) :!class /Script/Engine.MaterialInstanceDynamic)
      (return))))

(fn UpdateParticleSystemPosition
  (bind _particlescomponent (Variables|Business|GetParticlesComponent))
  (Utilities|IsValid _particlescomponent
    (:Is%20Valid
      (Transformation|SetWorldLocation _particlescomponent (CallFunction|GetSocketLocationWS (Variables|Default|GetSocket%20Attach%20Point)) false false))
    (:Is%20Not%20Valid
      _))
  (return))

(fn UpdateSplinePointPosition
  (if (Variables|Business|GetCanNewSplinePointsBeCreated)
    (Spline|SetLocationatSplinePoint (Variables|Components|GetSpline) (- (Spline|GetNumberOfSplinePoints (Variables|Components|GetSpline)) 1) (CallFunction|GetSocketLocationWS (Variables|Default|GetSocket%20Attach%20Point)) "World" true)))

(fn ConstructionScript
  (bind _socket_attach_point (Variables|Default|GetSocket%20Attach%20Point))
  (bind _location_ws (CallFunction|GetSocketLocationWS _socket_attach_point))
  (bind _colorandalphaoverlife (Variables|Curves|GetColorAndAlphaOverLife))
  (bind _sizeoverlife (Variables|Curves|GetSizeOverLife))
  (Networking|CanExecuteCosmeticEvents
    (:True
      (if (Variables|Debug|GetDebugMode)
        ; Debug Mode Stuff
        (Rendering|SetVisibility (Variables|Components|GetDebugArrow) true false)
        (bind _returnvalue (AddComponent|NODE_AddSkeletalMeshComponent-0))
        (bind _output_get (Variables|Components|SetSkeletal%20Mesh _returnvalue))
        (Transformation|AttachComponentToComponent _output_get (Variables|Components|GetDebugArrow) "None" "KeepRelative" "KeepRelative" "KeepRelative" true)
        (Transformation|AddLocalOffset _output_get (Variables|Debug|GetDebugPositionOffset) false false)
        (Variables|Default|SetSocket%20Attach%20Point "DebugSocket"))
      ;[1 1 1] Create Spline
      (bind _returnvalue_2 (AddComponent|NODE_AddSplineComponent-0 :bManualAttachment true))
      (bind _output_get_2 (Variables|Components|SetSpline _returnvalue_2))
      (Spline|SetDefaultUpVector _output_get_2 (Variables|Default|GetSplineDefaultUpVector) (select (Variables|Default|GetSplineDefaultUpVectorInWorldSpace) "World" "Local"))
      ;[1 1 1] Set both spline points to the socket's world position
      (Spline|SetLocationatSplinePoint (Variables|Components|GetSpline) 0 _location_ws "World" true)
      (Utilities|Array|Add (Variables|Business|GetSplineTimestamps) (Utilities|Time|GetGameTimeinSeconds))
      (Spline|SetLocationatSplinePoint (Variables|Components|GetSpline) 1 _location_ws "World" true)
      (Utilities|Array|Add (Variables|Business|GetSplineTimestamps) (Utilities|Time|GetGameTimeinSeconds))
      (CallFunction|AddSplineMesh)
      (Variables|Business|SetLifetimeAdjusted (Variables|Default|GetLifetime))
      ;[1 1 1] MID For baking blutility
      (if (Variables|CurveBaking|GetWriteBakedCurveDataToDisk)
        (bind _returnvalue_3 (Rendering|Material|CreateDynamicMaterialInstance "/Game/Effects/SplineTrails/Materials/M_Spline_BakeCurve_RT_01.M_Spline_BakeCurve_RT_01" "None" "None" :!class /Script/Engine.KismetMaterialLibrary))
        (Variables|CurveBaking|SetCurveBakeMID _returnvalue_3)
        ;[1 1 1] Curves to be used by the baking blutility
        (Utilities|Array|Clear (Variables|CurveBaking|GetLinearColorCurvesToBake))
        (Utilities|IsValid _colorandalphaoverlife
          (:Is%20Valid
            (Utilities|Array|Add (Variables|CurveBaking|GetLinearColorCurvesToBake) _colorandalphaoverlife))
          (:Is%20Not%20Valid
            _))
        (Utilities|IsValid _sizeoverlife
          (:Is%20Valid
            (Utilities|Array|Add (Variables|CurveBaking|GetLinearColorCurvesToBake) _sizeoverlife))
          (:Is%20Not%20Valid
            _))
        ;[1 1 1] If using tick interval LODs, check to make sure arrays are of equal lengths
        (if (and (Variables|LODs|GetUse%20Tick%20Interval%20LODs) (!= (Utilities|Array|Length (Variables|LODs|GetTick%20Interval%20LOD%20Distances)) (Utilities|Array|Length (Variables|LODs|GetTick%20Actor%20Intervals))))
          (Development|PrintString "Warning: Tick interval arrays are of different lengths, they must be equal." true true "(R=1.000000,G=0.000000,B=0.486624,A=1.000000)" 10))
        (else
          _)))))
