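# Unlit material in the SplineTrails effect set, named for baking spline curve data into a render target.
# Its only node is a LinearColor vector parameter with a magenta default, with no connection to a material output shown.
# It uses alpha composite blending and allows negative emissive color, so signed values can be output.
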Material /Game/Effects/SplineTrails/Materials/M_Spline_BakeCurve_RT_01
graph {
 properties {
  bAllowNegativeEmissiveColor=true
  blendMode=BLEND_AlphaComposite
  pixelDepthOffsetMode=PDOM_Legacy
  shadingModel=MSM_Unlit
  usedShadingModels=Unlit
 }
 nodes [VectorParameter() {parameterName=LinearColor, defaultValue=<rgba 0.947 0 1 1>}]
}
